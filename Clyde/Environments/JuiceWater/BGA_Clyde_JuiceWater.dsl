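; Juice water area actor that grants players gameplay effects, an overlap tag and a looping cue, mirroring effects and tag to carried DBNO players.
; Application requires actor-wide flags from both water-mask and reveal-radius overlaps and no vehicle tag; leaving either volume removes it.
; Construction and BeginPlay scale the radius collider and push radius, center, mask and SlurpActive values into material parameter collections.
; An optional debug view shows collision and radius.
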(event EventBeginPlay
  (CallFunction|UpdateWaterandCollision))

(event OnComponentBeginOverlap(SM_JuiceWaterCollision)
  (params (OverlappedComponent :type Primitive%20Component%20Object%20Reference) (OtherActor :type Actor%20Object%20Reference) (OtherComp :type Primitive%20Component%20Object%20Reference) (OtherBodyIndex :type Integer) (bFromSweep :type Boolean) (SweepResult :type Hit%20Result%20Structure%20%28by%20ref%29))
  (bind _asfort_player_pawn (Utilities|Casting|CastToFortPlayerPawn OtherActor))
  (Variables|Default|SetinJuiceMask true)
  (if (Variables|Default|GetinReveaRadius)
    (if (not (GameplayTags|HasMatchingGameplayTag _asfort_player_pawn (Variables|Default|GetVehicle%20Tag%20Check)))
      (Default|ApplySilkyGA _asfort_player_pawn)
      (Development|PrintString "Get Juice" true true "(R=1.000000,G=0.000000,B=0.690626,A=1.000000)" 1e+01))))

(event Custom|UpdateWaterAndCollision
  (bind _mpc (Variables|Default|GetMPC))
  (Transformation|SetWorldScale3D (Variables|Default|GetRadius) (Math|Conversions|ToVector(Float) (/ (Variables|Default|GetSlurpRevealRadius) 3e+01)))
  (Rendering|Material|SetScalarParameterValue (Variables|Default|GetMPC) (Variables|Default|GetMPC_Scalar_Radius) (Variables|Default|GetSlurpRevealRadius) :!class /Script/Engine.KismetMaterialLibrary)
  (Rendering|Material|SetVectorParameterValue (Variables|Default|GetMPC) (Variables|Default|GetMPC_Vector_Center) (Math|Conversions|ToLinearColor(Vector) (Transformation|GetWorldLocation (Variables|Default|GetDefaultSceneRoot))) :!class /Script/Engine.KismetMaterialLibrary)
  (Rendering|Material|SetVectorParameterValue _mpc (Variables|Default|GetTextureMaskLocation) (Math|Conversions|ToLinearColor(Vector) (Variables|Default|GetMaskLocation)) :!class /Script/Engine.KismetMaterialLibrary)
  (Rendering|Material|SetVectorParameterValue _mpc (Variables|Default|GetJuiceMaskSize) (Utilities|Struct|MakeLinearColor (Variables|Default|GetMaskTextureSize) (Variables|Default|GetMaskTextureSize) (Variables|Default|GetMaskTextureSize)) :!class /Script/Engine.KismetMaterialLibrary)
  (Rendering|Material|SetScalarParameterValue "/Game/Packages/Fortress_SharedMaterials/GlobalMaterialParameters/FortniteMaterialParameters.FortniteMaterialParameters" "SlurpActive" (Variables|Default|GetSlurp%20Active) :!class /Script/Engine.KismetMaterialLibrary))

(event OnComponentEndOverlap(SM_JuiceWaterCollision)
  (params (OverlappedComponent :type Primitive%20Component%20Object%20Reference) (OtherActor :type Actor%20Object%20Reference) (OtherComp :type Primitive%20Component%20Object%20Reference) (OtherBodyIndex :type Integer))
  (bind _asfort_player_pawn (Utilities|Casting|CastToFortPlayerPawn OtherActor))
  (Variables|Default|SetinJuiceMask false)
  (Default|RemoveSilkyGA _asfort_player_pawn)
  (Development|PrintString "Remove Juice" true true "(R=1.000000,G=0.000000,B=0.690626,A=1.000000)" 1e+01))

(event Custom|DBNOHoistChange
  (params (CarryEvent :type EFortDBNOCarryEvent%20Enum :default "PickedUp") (HoisterPawn :type Fort%20Player%20Pawn%20Object%20Reference) (HoisteePawn :type Fort%20Player%20Pawn%20Object%20Reference))
  (GameplayEffects|RemoveActiveEffectswithTags (Ability|GetAbilitySystemComponent HoisteePawn) (Variables|Default|GetActiveEffectTag))
  (Ability|GameplayEffect|RemoveLooseGameplayTags HoisteePawn (Variables|Default|GetTag_Overlapping) (Variables|Default|GetTag_Overlapping_Should_Replicate))
  (EventDispatchers|UnbindEventfromOnDBNOHoisterChangedDelegate HoisteePawn (delegate Custom|DBNOHoistChange)))

(event OnComponentBeginOverlap(Radius)
  (params (OverlappedComponent :type Primitive%20Component%20Object%20Reference) (OtherActor :type Actor%20Object%20Reference) (OtherComp :type Primitive%20Component%20Object%20Reference) (OtherBodyIndex :type Integer) (bFromSweep :type Boolean) (SweepResult :type Hit%20Result%20Structure%20%28by%20ref%29))
  (bind _asfort_player_pawn (Utilities|Casting|CastToFortPlayerPawn OtherActor))
  (Variables|Default|SetinReveaRadius true)
  (if (Variables|Default|GetinJuiceMask)
    (if (not (GameplayTags|HasMatchingGameplayTag _asfort_player_pawn (Variables|Default|GetVehicle%20Tag%20Check)))
      (Default|ApplySilkyGA _asfort_player_pawn)
      (Development|PrintString "Get Juice" true true "(R=1.000000,G=0.000000,B=0.690626,A=1.000000)" 1e+01))))

(event OnComponentEndOverlap(Radius)
  (params (OverlappedComponent :type Primitive%20Component%20Object%20Reference) (OtherActor :type Actor%20Object%20Reference) (OtherComp :type Primitive%20Component%20Object%20Reference) (OtherBodyIndex :type Integer))
  (bind _asfort_player_pawn (Utilities|Casting|CastToFortPlayerPawn OtherActor))
  (Variables|Default|SetinReveaRadius false)
  (Default|RemoveSilkyGA _asfort_player_pawn)
  (Development|PrintString "Remove Juice" true true "(R=1.000000,G=0.000000,B=0.690626,A=1.000000)" 1e+01))

(event Custom|JuiceWater
  (Fortnite|ApplyGameplayEffectsToActor (Game|GetPlayerPawn 0) (Variables|Default|GetGE_To_Apply) (Transformation|GetActorLocation (Game|GetPlayerPawn 0)))
  (Development|PrintString "Applying da JUICE"))

(fn ApplySilkyGA
  (params (InActor :type Actor%20Object%20Reference))
  (bind _asfort_pawn (Utilities|Casting|CastToFortPawn InActor))
  (bind _output_get (Variables|Default|SetFort%20Pawn _asfort_pawn))
  ;[1 1 1] Check pawn is correctly overlapping the water, if so give the pawn the In Water Tag
  (Ability|GameplayEffect|AddLooseGameplayTags _output_get (Variables|Default|GetTag_Overlapping) (Variables|Default|GetTag_Overlapping_Should_Replicate))
  (for _array_element (Variables|Default|GetGE_To_Apply)
    (GameplayEffects|FortApplyGameplayEffectToTarget :self (Ability|GetFortAbilitySystemComponent self) :GameplayEffectClass _array_element :TargetOfEffect (Ability|GetAbilitySystemComponent (Variables|Default|GetFort%20Pawn)) :OptionalSourceObject self :OptionalEffectCauser self :Level 0.0)
    (GameplayCue|AddGameplayCueOnActor(Looping) (Variables|Default|GetFort%20Pawn) (Variables|Default|GetGameplay%20Cue%20Tag) (Ability|GameplayCue|MakeGameplayCueParameters 0.0))
    ;[1 1 1] Apply to carried DBNO player
    (bind _asfort_player_pawn (Utilities|Casting|CastToFortPlayerPawn (Variables|Default|GetFort%20Pawn)))
    (Utilities|IsValid (Class|FortPlayerPawn|GetDBNOHoistee _asfort_player_pawn)
      (:Is%20Valid
        (Ability|GameplayEffect|AddLooseGameplayTags (Class|FortPlayerPawn|GetDBNOHoistee _asfort_player_pawn) (Variables|Default|GetTag_Overlapping) (Variables|Default|GetTag_Overlapping_Should_Replicate))
        (for _array_element_2 (Variables|Default|GetGE_To_Apply)
          (GameplayEffects|FortApplyGameplayEffectToTarget :self (Ability|GetFortAbilitySystemComponent (Class|FortPlayerPawn|GetDBNOHoistee _asfort_player_pawn)) :GameplayEffectClass _array_element_2 :TargetOfEffect (Ability|GetAbilitySystemComponent (Class|FortPlayerPawn|GetDBNOHoistee _asfort_player_pawn)) :OptionalSourceObject (Class|FortPlayerPawn|GetDBNOHoistee _asfort_player_pawn) :OptionalEffectCauser (Class|FortPlayerPawn|GetDBNOHoistee _asfort_player_pawn) :Level 0.0)
          (EventDispatchers|BindEventtoOnDBNOHoisterChangedDelegate (Class|FortPlayerPawn|GetDBNOHoistee _asfort_player_pawn) (EventDispatchers|CreateEvent :!event DBNOHoistChange)))))))

(fn RemoveSilkyGA
  (params (InActor :type Actor%20Object%20Reference))
  (bind _asfort_pawn (Utilities|Casting|CastToFortPawn InActor))
  (bind _output_get (Variables|Default|SetFort%20Pawn _asfort_pawn))
  (Ability|GameplayEffect|RemoveLooseGameplayTags _output_get (Variables|Default|GetTag_Overlapping) (Variables|Default|GetTag_Overlapping_Should_Replicate))
  (GameplayCue|RemoveGameplayCueOnActor(Looping) _output_get (Variables|Default|GetGameplay%20Cue%20Tag) (Ability|GameplayCue|MakeGameplayCueParameters 0.0))
  (GameplayEffects|RemoveActiveEffectswithTags (Ability|GetAbilitySystemComponent _output_get) (Variables|Default|GetActiveEffectTag))
  ;[1 1 1] Apply to carried DBNO player
  (bind _asfort_player_pawn (Utilities|Casting|CastToFortPlayerPawn (Variables|Default|GetFort%20Pawn)))
  (Utilities|IsValid (Class|FortPlayerPawn|GetDBNOHoistee _asfort_player_pawn)
    (:Is%20Valid
      (Ability|GameplayEffect|RemoveLooseGameplayTags (Class|FortPlayerPawn|GetDBNOHoistee _asfort_player_pawn) (Variables|Default|GetTag_Overlapping) (Variables|Default|GetTag_Overlapping_Should_Replicate))
      (GameplayEffects|RemoveActiveEffectswithTags (Ability|GetAbilitySystemComponent (Class|FortPlayerPawn|GetDBNOHoistee _asfort_player_pawn)) (Variables|Default|GetActiveEffectTag)))))

(fn ConstructionScript
  (bind _sm_juicewatercollisiondebug (Variables|Default|GetSM_JuiceWaterCollisionDebug))
  (CallFunction|UpdateWaterandCollision)
  (if (Variables|Default|GetShowDebug)
    (Rendering|SetVisibility _sm_juicewatercollisiondebug true false)
    (Rendering|Debug|DrawDebugCylinder (Transformation|GetWorldLocation (Variables|Default|GetDefaultSceneRoot)) (+ (Transformation|GetWorldLocation (Variables|Default|GetDefaultSceneRoot)) "0.000000,0.000000,1500.000000") (Variables|Default|GetSlurpRevealRadius) 12 "(R=0.001967,G=0.488364,B=1.000000,A=1.000000)" 0.5 1e+02)
    (else
      (Rendering|SetVisibility _sm_juicewatercollisiondebug false false))))
