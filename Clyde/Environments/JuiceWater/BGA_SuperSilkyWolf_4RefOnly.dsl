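; Slurp water actor that applies a gameplay effect, gated by a heal setting, to pawns overlapping its water.
; BeginPlay creates a dynamic water material fed by the fluid sim's normal render target, refreshed on water scalability changes.
; Once playlist data is ready, overlap events add or remove tags and the effect.
; Carried DBNO players receive the same treatment, cleared when their hoister changes.
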(event Custom|ReCheckScalability
  ;[1 1 1] Support Changing Scalability
  (if (Water|IsShallowWaterSimulationEnabled (WorldSubsystems|GetWaterSubsystem))
    (Utilities|FlowControl|Delay 2.0)
    (bind _outactors (Actor|GetAllActorsOfClass "/Game/Athena/Apollo/Water/Blueprints/BP_FluidSim_FN.BP_FluidSim_FN_C"))
    (Utilities|IsValid (Utilities|Array|Get(acopy) _outactors)
      (:Is%20Valid
        (Utilities|IsValid (Variables|/FortWater/FluidSimulation/Blueprints/BP_FluidSim_01.BP_FluidSim_01_C|GetNormalRT (Utilities|Array|Get(acopy) _outactors))
          (:Is%20Valid
            (Rendering|Material|SetTextureParameterValue (Variables|Default|GetSlurp%20MID) "NormalAndHeight" (Variables|/FortWater/FluidSimulation/Blueprints/BP_FluidSim_01.BP_FluidSim_01_C|GetNormalRT (Utilities|Array|Get(acopy) _outactors))))
          (:Is%20Not%20Valid
            (Development|PrintString (Utilities|String|Append "Water log: Fluidsim " (Utilities|GetDisplayName (Utilities|Array|Get(acopy) _outactors)) "had invalid Normal RT and there were: " (Utilities|String|ToString(Integer) (Utilities|Array|Length _outactors)) "fluidsims found.")))))
      (:Is%20Not%20Valid
        (Development|PrintString "No Fluid Sim Ruh-Roh")))
    (else
      (Rendering|Material|SetTextureParameterValue (Variables|Default|GetSlurp%20MID) "NormalAndHeight" "/Engine/EngineResources/Black_Low.Black"))))

(event Custom|OnComponentEndOverlap_Event_0
  (params (OverlappedComponent :type Primitive%20Component%20Object%20Reference) (OtherActor :type Actor%20Object%20Reference) (OtherComp :type Primitive%20Component%20Object%20Reference) (OtherBodyIndex :type Integer))
  ;[1 1 1] Bind on end overlap, remove gameplay effect from pawn
  (CallFunction|RemoveSilkyGA OtherActor))

(event Custom|OnComponentBeginOverlap_Event_0
  (params (OverlappedComponent :type Primitive%20Component%20Object%20Reference) (OtherActor :type Actor%20Object%20Reference) (OtherComp :type Primitive%20Component%20Object%20Reference) (OtherBodyIndex :type Integer) (bFromSweep :type Boolean) (SweepResult :type Hit%20Result%20Structure%20%28by%20ref%29))
  (CallFunction|ApplySilkyGA OtherActor))

(event Custom|DBNOHoistChange
  (params (CarryEvent :type EFortDBNOCarryEvent%20Enum :default "PickedUp") (HoisterPawn :type Fort%20Player%20Pawn%20Object%20Reference) (HoisteePawn :type Fort%20Player%20Pawn%20Object%20Reference))
  (GameplayEffects|RemoveActiveEffectswithTags (Ability|GetAbilitySystemComponent HoisteePawn) (Variables|Default|GetActiveEffectTag))
  (Ability|GameplayEffect|RemoveLooseGameplayTags HoisteePawn (Variables|Default|GetTag_Overlapping) (Variables|Default|GetTag_Overlapping_Should_Replicate))
  (EventDispatchers|UnbindEventfromOnDBNOHoisterChangedDelegate HoisteePawn (delegate Custom|DBNOHoistChange)))

(event EventBeginPlay
  (Water|BindEventtoOnWaterScalabilityChanged (WorldSubsystems|GetWaterSubsystem) (delegate Custom|ReCheckScalability))
  ;[1 1 1] Setup Correct Mesh for Asset
  (bind _returnvalue (Rendering|Material|CreateDynamicMaterialInstance (Variables|Default|GetWater) 0 (Rendering|Material|GetMaterial (Variables|Default|GetWater) 0) (Utilities|String|StringToName (Utilities|String|Append "Slurp_MID_" (Utilities|String|ToString(Object) (Rendering|Material|GetMaterial (Variables|Default|GetWater) 0)))) :!class /Script/Engine.PrimitiveComponent))
  (bind _output_get (Variables|Default|SetSlurp%20MID _returnvalue))
  (Rendering|Material|SetMaterial (Variables|Default|GetWater) 0 _output_get)
  ;[1 1 1] Fluid Sim Addition
  (bind _outactors (Actor|GetAllActorsOfClass "/Game/Athena/Apollo/Water/Blueprints/BP_FluidSim_FN.BP_FluidSim_FN_C"))
  (if (> (Utilities|Array|Length _outactors) 0)
    (Utilities|IsValid (Utilities|Array|Get(acopy) _outactors)
      (:Is%20Valid
        (Rendering|Material|SetTextureParameterValue (Variables|Default|GetSlurp%20MID) "NormalAndHeight" (Variables|/FortWater/FluidSimulation/Blueprints/BP_FluidSim_01.BP_FluidSim_01_C|GetNormalRT (Utilities|Array|Get(acopy) _outactors))))
      (:Is%20Not%20Valid
        (Development|PrintString "No Fluid Sim Ruh-Roh")))
    (Class|FortAsyncActionCurrentPlaylistReady|PlaylistDataReadyAsync
      (:OnReady
        (Utilities|FlowControl|Sequence
          (:then_0
            (Utilities|IsValid (Variables|Default|GetWater)
              (:Is%20Valid
                (if (> (ScalableFloat|GetValueatLevel (Variables|Default|GetShouldApplyHeal)) 0)
                  (Collision|AssignOnComponentBeginOverlap (Variables|Default|GetWater) (delegate Custom|OnComponentBeginOverlap_Event_0))
                  ;[1 1 1] Bind on end overlap, remove gameplay effect from pawn
                  (Collision|AssignOnComponentEndOverlap (Variables|Default|GetWater) (delegate Custom|OnComponentEndOverlap_Event_0)))))))))))

(fn ApplySilkyGA
  (params (InActor :type Actor%20Object%20Reference))
  (bind _asfort_pawn (Utilities|Casting|CastToFortPawn InActor))
  (bind _output_get (Variables|Default|SetFort%20Pawn _asfort_pawn))
  ;[1 1 1] Check pawn is correctly overlapping the water, if so give the pawn the In Water Tag
  (Ability|GameplayEffect|AddLooseGameplayTags _output_get (Variables|Default|GetTag_Overlapping) (Variables|Default|GetTag_Overlapping_Should_Replicate))
  (GameplayEffects|FortApplyGameplayEffectToTarget :self (Ability|GetFortAbilitySystemComponent self) :GameplayEffectClass (Variables|Default|GetGE_To_Apply) :TargetOfEffect (Ability|GetAbilitySystemComponent (Variables|Default|GetFort%20Pawn)) :OptionalSourceObject self :OptionalEffectCauser self :Level 0.0)
  ;[1 1 1] Apply to carried DBNO player
  (bind _asfort_player_pawn (Utilities|Casting|CastToFortPlayerPawn (Variables|Default|GetFort%20Pawn)))
  (Utilities|IsValid (Class|FortPlayerPawn|GetDBNOHoistee _asfort_player_pawn)
    (:Is%20Valid
      (Ability|GameplayEffect|AddLooseGameplayTags (Class|FortPlayerPawn|GetDBNOHoistee _asfort_player_pawn) (Variables|Default|GetTag_Overlapping) (Variables|Default|GetTag_Overlapping_Should_Replicate))
      (GameplayEffects|FortApplyGameplayEffectToTarget :self (Ability|GetFortAbilitySystemComponent (Class|FortPlayerPawn|GetDBNOHoistee _asfort_player_pawn)) :GameplayEffectClass (Variables|Default|GetGE_To_Apply) :TargetOfEffect (Ability|GetAbilitySystemComponent (Class|FortPlayerPawn|GetDBNOHoistee _asfort_player_pawn)) :OptionalSourceObject (Class|FortPlayerPawn|GetDBNOHoistee _asfort_player_pawn) :OptionalEffectCauser (Class|FortPlayerPawn|GetDBNOHoistee _asfort_player_pawn) :Level 0.0)
      (EventDispatchers|BindEventtoOnDBNOHoisterChangedDelegate (Class|FortPlayerPawn|GetDBNOHoistee _asfort_player_pawn) (EventDispatchers|CreateEvent :!event DBNOHoistChange)))))

(fn RemoveSilkyGA
  (params (InActor :type Actor%20Object%20Reference))
  (bind _asfort_pawn (Utilities|Casting|CastToFortPawn InActor))
  (bind _output_get (Variables|Default|SetFort%20Pawn _asfort_pawn))
  (Ability|GameplayEffect|RemoveLooseGameplayTags _output_get (Variables|Default|GetTag_Overlapping) (Variables|Default|GetTag_Overlapping_Should_Replicate))
  (GameplayEffects|RemoveActiveEffectswithTags (Ability|GetAbilitySystemComponent _output_get) (Variables|Default|GetActiveEffectTag))
  ;[1 1 1] Apply to carried DBNO player
  (bind _asfort_player_pawn (Utilities|Casting|CastToFortPlayerPawn (Variables|Default|GetFort%20Pawn)))
  (Utilities|IsValid (Class|FortPlayerPawn|GetDBNOHoistee _asfort_player_pawn)
    (:Is%20Valid
      (Ability|GameplayEffect|RemoveLooseGameplayTags (Class|FortPlayerPawn|GetDBNOHoistee _asfort_player_pawn) (Variables|Default|GetTag_Overlapping) (Variables|Default|GetTag_Overlapping_Should_Replicate))
      (GameplayEffects|RemoveActiveEffectswithTags (Ability|GetAbilitySystemComponent (Class|FortPlayerPawn|GetDBNOHoistee _asfort_player_pawn)) (Variables|Default|GetActiveEffectTag)))))

(fn ConstructionScript)
