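; Kelp Linen event actor that streams staged world data layers on clients and applies ground effects to player pawns.
; Calendar events set the stage index, whose RepNotify loads that stage's layers if the local pawn is inside the streaming volume, retrying on a timer.
; Volume overlaps also load or unload layers.
; Ground GameplayEffects or cues require two overlap flags, shared actor-wide; one capsule follows the global SlurpyCenter and SlurpyRadius material parameters.
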(event EventBeginPlay
  (Class|FortAsyncActionCurrentPlaylistReady|PlaylistDataReadyAsync
    (:OnReady
      (CallFunction|SetupVolumes)
      (bind _eventname (MCPContext|WatchCalendarEvents (Variables|Default|GetCalendarEventsToWatch))
        (:OnEventActive
          (switch string _eventname
            (:KL1
              (Variables|Default|SetKelp%20Linen%20Index 0))
            (:KL2
              (Variables|Default|SetKelp%20Linen%20Index 1))
            (:KL3
              (Variables|Default|SetKelp%20Linen%20Index 2)))
          (if (Replay|GetIsPlayingReplay)
            (Networking|CanExecuteCosmeticEvents
              (:True
                (CallFunction|LoadClientKelpLinenDatalayers (Variables|Default|GetKelp%20Linen%20Local%20Index))
                (Fortnite|FortShippingLog self "Loaded Kelp Linen Client For Replay")))))))))

(event OnComponentBeginOverlap(SM_KelpLinen_GroundFXVolume)
  (params (OverlappedComponent :type Primitive%20Component%20Object%20Reference) (OtherActor :type Actor%20Object%20Reference) (OtherComp :type Primitive%20Component%20Object%20Reference) (OtherBodyIndex :type Integer) (bFromSweep :type Boolean) (SweepResult :type Hit%20Result%20Structure%20%28by%20ref%29))
  ; GroundFX
  (bind _asfort_player_pawn (Utilities|Casting|CastToFortPlayerPawn OtherActor))
  (Variables|Default|SetInPurpleGrass true)
  (CallFunction|ApplyGroundFXGCN _asfort_player_pawn))

(event OnComponentEndOverlap(SM_KelpLinen_GroundFXVolume)
  (params (OverlappedComponent :type Primitive%20Component%20Object%20Reference) (OtherActor :type Actor%20Object%20Reference) (OtherComp :type Primitive%20Component%20Object%20Reference) (OtherBodyIndex :type Integer))
  ; GroundFX
  (bind _asfort_player_pawn (Utilities|Casting|CastToFortPlayerPawn OtherActor))
  (Variables|Default|SetInPurpleGrass false)
  (CallFunction|RemoveGroundFXGCN _asfort_player_pawn))

(event OnComponentBeginOverlap(GroundFXOverlap)
  (params (OverlappedComponent :type Primitive%20Component%20Object%20Reference) (OtherActor :type Actor%20Object%20Reference) (OtherComp :type Primitive%20Component%20Object%20Reference) (OtherBodyIndex :type Integer) (bFromSweep :type Boolean) (SweepResult :type Hit%20Result%20Structure%20%28by%20ref%29))
  ; GroundFX
  (bind _asfort_player_pawn (Utilities|Casting|CastToFortPlayerPawn OtherActor))
  (Variables|Default|SetInRevealRadius true)
  (CallFunction|ApplyGroundFXGCN _asfort_player_pawn))

(event OnComponentEndOverlap(GroundFXOverlap)
  (params (OverlappedComponent :type Primitive%20Component%20Object%20Reference) (OtherActor :type Actor%20Object%20Reference) (OtherComp :type Primitive%20Component%20Object%20Reference) (OtherBodyIndex :type Integer))
  ; GroundFX
  (bind _asfort_player_pawn (Utilities|Casting|CastToFortPlayerPawn OtherActor))
  (Variables|Default|SetInRevealRadius false)
  (CallFunction|RemoveGroundFXGCN _asfort_player_pawn))

(event Custom|RetryBeginPlayClientLoad
  ; LocalClientStreaming
  (if (== (Utilities|Array|Get(acopy) (Collision|GetOverlappingActors (Variables|Default|GetLocalStreamingOverlap) "/Script/FortniteGame.FortPlayerPawn" :!class /Script/Engine.PrimitiveComponent) (Utilities|Array|FindItem (Collision|GetOverlappingActors (Variables|Default|GetLocalStreamingOverlap) "/Script/FortniteGame.FortPlayerPawn" :!class /Script/Engine.PrimitiveComponent) (Game|Player|GetPlayerPawn (Fortnite|GetFirstLocalFortPlayerController true)))) (Game|Player|GetPlayerPawn (Fortnite|GetFirstLocalFortPlayerController true)))
    (CallFunction|LoadClientKelpLinenDatalayers (Variables|Default|GetKelp%20Linen%20Local%20Index))
    (Utilities|Time|ClearandInvalidateTimerbyHandle (Variables|Default|GetBeginPlayLoadTimer))
    (else
      (bind _returnvalue (Utilities|Time|SetTimerbyEvent (EventDispatchers|CreateEvent :!event Custom|RetryBeginPlayClientLoad) 1.0 true))
      (Variables|Default|SetBeginPlayLoadTimer _returnvalue)
      (bind _result (Math|Integer|IncrementInt (Variables|Default|GetNumberOfRetries))
        (:%20%20
          (bind _output_get (Variables|Default|SetNumberOfRetries _result))
          (if (>= _output_get (Variables|Default|GetMaxRetries))
            (Utilities|Time|ClearandInvalidateTimerbyHandle (Variables|Default|GetBeginPlayLoadTimer))))))))

(event Custom|LoadClientKelpLinenBeginPlay
  ; LocalClientStreaming
  (Networking|CanExecuteCosmeticEvents
    (:True
      (if (== (Utilities|Array|Get(acopy) (Collision|GetOverlappingActors (Variables|Default|GetLocalStreamingOverlap) "/Script/FortniteGame.FortPlayerPawn" :!class /Script/Engine.PrimitiveComponent) (Utilities|Array|FindItem (Collision|GetOverlappingActors (Variables|Default|GetLocalStreamingOverlap) "/Script/FortniteGame.FortPlayerPawn" :!class /Script/Engine.PrimitiveComponent) (Game|Player|GetPlayerPawn (Fortnite|GetFirstLocalFortPlayerController true)))) (Game|Player|GetPlayerPawn (Fortnite|GetFirstLocalFortPlayerController true)))
        (CallFunction|LoadClientKelpLinenDatalayers (Variables|Default|GetKelp%20Linen%20Local%20Index))
        (Utilities|Time|ClearandInvalidateTimerbyHandle (Variables|Default|GetBeginPlayLoadTimer))
        (else
          (bind _returnvalue (Utilities|Time|SetTimerbyEvent (EventDispatchers|CreateEvent :!event Custom|RetryBeginPlayClientLoad) 1.0 true))
          (Variables|Default|SetBeginPlayLoadTimer _returnvalue)
          (bind _result (Math|Integer|IncrementInt (Variables|Default|GetNumberOfRetries))
            (:%20%20
              (bind _output_get (Variables|Default|SetNumberOfRetries _result))
              (if (>= _output_get (Variables|Default|GetMaxRetries))
                (Utilities|Time|ClearandInvalidateTimerbyHandle (Variables|Default|GetBeginPlayLoadTimer))))))))))

(event OnComponentBeginOverlap(LocalStreamingOverlap)
  (params (OverlappedComponent :type Primitive%20Component%20Object%20Reference) (OtherActor :type Actor%20Object%20Reference) (OtherComp :type Primitive%20Component%20Object%20Reference) (OtherBodyIndex :type Integer) (bFromSweep :type Boolean) (SweepResult :type Hit%20Result%20Structure%20%28by%20ref%29))
  ; LocalClientStreaming
  (bind _asfort_player_pawn (Utilities|Casting|CastToFortPlayerPawn OtherActor))
  (Networking|CanExecuteCosmeticEvents
    (:True
      (if (== (Game|Player|GetPlayerPawn (Fortnite|GetFirstLocalFortPlayerController true)) _asfort_player_pawn)
        (CallFunction|LoadClientKelpLinenDatalayers (Variables|Default|GetKelp%20Linen%20Local%20Index))))))

(event OnComponentEndOverlap(LocalStreamingOverlap)
  (params (OverlappedComponent :type Primitive%20Component%20Object%20Reference) (OtherActor :type Actor%20Object%20Reference) (OtherComp :type Primitive%20Component%20Object%20Reference) (OtherBodyIndex :type Integer))
  ; LocalClientStreaming
  (bind _asfort_player_pawn (Utilities|Casting|CastToFortPlayerPawn OtherActor))
  (if (not (Class|FortPawn|IsInsideHidingProp _asfort_player_pawn true))
    (Networking|CanExecuteCosmeticEvents
      (:True
        (if (== (Game|Player|GetPlayerPawn (Fortnite|GetFirstLocalFortPlayerController true)) _asfort_player_pawn)
          (CallFunction|UnloadAllClientKelpLinenDataLayers))))))

(fn ApplyGroundFXGCN
  (params (Target :type Actor%20Object%20Reference))
  (bind _gameplay_cue_tag (Variables|Default|GetGameplay%20Cue%20Tag))
  (if (and (Variables|Default|GetInPurpleGrass) (Variables|Default|GetInRevealRadius))
    (if (Variables|Default|GetUseGE)
      (switch int (Variables|Default|GetKelp%20Linen%20Index)
        (:0
          (GameplayEffects|ApplyGameplayEffectToTarget :self (Ability|GetFortAbilitySystemComponent Target) :GameplayEffectClass "/Clyde/GameplayEffects/KelpLInen/GE_KelpLinenGroundFX.GE_KelpLinenGroundFX_C" :Target (Ability|GetFortAbilitySystemComponent Target) :Level 0.0)
          (Development|PrintString "Applied Kelp GE Index 0"))
        (:1
          (GameplayEffects|ApplyGameplayEffectToTarget :self (Ability|GetFortAbilitySystemComponent Target) :GameplayEffectClass "/Clyde/GameplayEffects/KelpLInen/GE_KelpLinenGroundFX_Index1.GE_KelpLinenGroundFX_Index1_C" :Target (Ability|GetFortAbilitySystemComponent Target) :Level 0.0)
          (Development|PrintString "Applied Kelp GE Index 1"))
        (:2
          (GameplayEffects|ApplyGameplayEffectToTarget :self (Ability|GetFortAbilitySystemComponent Target) :GameplayEffectClass "/Clyde/GameplayEffects/KelpLInen/GE_KelpLinenGroundFX_Index2.GE_KelpLinenGroundFX_Index2_C" :Target (Ability|GetFortAbilitySystemComponent Target) :Level 0.0)
          (Development|PrintString "Applied Kelp GE Index 2")))
      (else
        (GameplayCue|AddGameplayCueOnActor(Looping) Target _gameplay_cue_tag (Variables|Default|GetParameters))
        (GameplayCue|ExecuteGameplayCue Target _gameplay_cue_tag)
        (Development|PrintString "Applied Kelp GCNL")))))

(fn LoadClientKelpLinenDatalayers
  (params (KelpLinenIndex :type Integer))
  (bind _returnvalue (WorldPartition|GetDataLayerManager))
  (switch int KelpLinenIndex
    (:0
      ; Week 4 - Day 1
      (for _array_element (Variables|Default|GetKelpLinen_Day1_ToLoad)
        (DataLayers|SetDataLayerInstanceRuntimeState _returnvalue (DataLayers|GetDataLayerInstancefromAsset _returnvalue _array_element) "Activated" false))
      (for _array_element_2 (Variables|Default|GetKelpLinen_Day1_ToUnload)
        (DataLayers|SetDataLayerInstanceRuntimeState _returnvalue (DataLayers|GetDataLayerInstancefromAsset _returnvalue _array_element_2) "Unloaded" false)))
    (:1
      ; Week 4 - Day 3
      (for _array_element_3 (Variables|Default|GetKelpLinen_Day3_ToLoad)
        (DataLayers|SetDataLayerInstanceRuntimeState _returnvalue (DataLayers|GetDataLayerInstancefromAsset _returnvalue _array_element_3) "Activated" false))
      (for _array_element_4 (Variables|Default|GetKelpLinen_Day3_ToUnload)
        (DataLayers|SetDataLayerInstanceRuntimeState _returnvalue (DataLayers|GetDataLayerInstancefromAsset _returnvalue _array_element_4) "Unloaded" false)))
    (:2
      ; Week 4 - Day 5
      (for _array_element_5 (Variables|Default|GetKelpLinen_Day5_ToLoad)
        (DataLayers|SetDataLayerInstanceRuntimeState _returnvalue (DataLayers|GetDataLayerInstancefromAsset _returnvalue _array_element_5) "Activated" false))
      (for _array_element_6 (Variables|Default|GetKelpLinen_Day5_ToUnload)
        (DataLayers|SetDataLayerInstanceRuntimeState _returnvalue (DataLayers|GetDataLayerInstancefromAsset _returnvalue _array_element_6) "Unloaded" false)))))

(fn OnRep_KelpLinenIndex
  (Networking|SwitchHasAuthority
    (:Remote
      (bind _output_get (Variables|Default|SetKelp%20Linen%20Local%20Index (Variables|Default|GetKelp%20Linen%20Index)))
      (Development|PrintString (Utilities|String|ToString(Integer) _output_get) true true "(R=1.000000,G=0.000000,B=0.725624,A=1.000000)" 5e+01)
      (CallFunction|LoadClientKelpLinenBeginPlay))))

(fn RemoveGroundFXGCN
  (params (Target :type Actor%20Object%20Reference))
  (if (Variables|Default|GetUseGE)
    (GameplayEffects|RemoveActiveEffectswithSourceTags (Ability|GetFortAbilitySystemComponent Target) "(GameplayTags=((TagName=\"Clyde.KelpLinen.FXApplied\")))")
    (Development|PrintString "Removed Kelp GE")
    (else
      (GameplayCue|RemoveGameplayCueOnActor(Looping) Target (Variables|Default|GetGameplay%20Cue%20Tag) (Variables|Default|GetParameters))
      (Development|PrintString "Removed Kelp GCNL"))))

(fn SetupVolumes
  (bind _groundfxoverlap (Variables|Default|GetGroundFXOverlap))
  (bind _ground_fx_radius_buffer (Variables|Default|GetGround%20FX%20radius%20buffer))
  (bind _returnvalue (Rendering|Material|GetVectorParameterValue "/Game/Packages/Fortress_SharedMaterials/GlobalMaterialParameters/FortniteMaterialParameters.FortniteMaterialParameters" "SlurpyCenter" :!class /Script/Engine.KismetMaterialLibrary))
  (bind (_returnvalue_r _returnvalue_g _returnvalue_b) (Math|Color|BreakColor _returnvalue))
  (Transformation|SetWorldLocation _groundfxoverlap (Math|Vector|MakeVector _returnvalue_r _returnvalue_g _returnvalue_b) false false)
  (bind _returnvalue_2 (Rendering|Material|GetScalarParameterValue "/Game/Packages/Fortress_SharedMaterials/GlobalMaterialParameters/FortniteMaterialParameters.FortniteMaterialParameters" "SlurpyRadius" :!class /Script/Engine.KismetMaterialLibrary))
  (Components|Capsule|SetCapsuleRadius _groundfxoverlap (+ _returnvalue_2 _ground_fx_radius_buffer) true)
  (if (Variables|Default|GetShowDebug)
    (Rendering|Debug|DrawDebugCapsule (Math|Vector|MakeVector _returnvalue_r _returnvalue_g _returnvalue_b) 1.5e+04 (+ _returnvalue_2 _ground_fx_radius_buffer) "0, 0, 0" "(R=0.971525,G=1.000000,B=0.010411,A=1.000000)" 1e+01 5e+01)
    (Rendering|Debug|DrawDebugCapsule (Transformation|GetWorldLocation (Variables|Default|GetLocalStreamingOverlap)) (Variables|/Script/Engine.CapsuleComponent|GetCapsuleHalfHeight (Variables|Default|GetLocalStreamingOverlap)) (Variables|/Script/Engine.CapsuleComponent|GetCapsuleRadius (Variables|Default|GetLocalStreamingOverlap)) "0, 0, 0" "(R=0.924506,G=0.000000,B=1.000000,A=1.000000)" 1e+01 2.5e+02)))

(fn UnloadAllClientKelpLinenDataLayers
  (bind _returnvalue (WorldPartition|GetDataLayerManager))
  (for _array_element (Variables|Default|GetKelpLinen_Client_ToUnload)
    (DataLayers|SetDataLayerInstanceRuntimeState _returnvalue (DataLayers|GetDataLayerInstancefromAsset _returnvalue _array_element) "Unloaded" false)))

(fn ConstructionScript
  (CallFunction|SetupVolumes))
